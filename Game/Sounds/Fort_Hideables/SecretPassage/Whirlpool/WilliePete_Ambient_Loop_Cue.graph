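# Sound cue for the looping WilliePete ambient sound, filed under the whirlpool secret passage hideable and assigned to the gadgets and items sound class.
# A looping wave player feeds an attenuation node that overrides distance falloff with low-pass filtering, stereo spread and listener focus that lowers non-focused volume, with reverb send off.
# The cue primes on load and overrides concurrency to cap simultaneous instances, stopping the quietest.
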SoundCue /Game/Sounds/Fort_Hideables/SecretPassage/Whirlpool/WilliePete_Ambient_Loop_Cue
root=Attenuation_1
nodes [
 {
  id=Attenuation_1
  children [WavePlayer_0]
  properties {
   bOverrideAttenuation=true
   attenuationOverrides {
    bAttenuateWithLPF=true
    bEnableFocusInterpolation=true
    bEnableListenerFocus=true
    bEnableReverbSend=false
    falloffDistance=3000
    focusReleaseInterpSpeed=0.5
    lPFFrequencyAtMax=2000
    lPFRadiusMax=3000
    lPFRadiusMin=500
    nonFocusVolumeAttenuation=0.5
    stereoSpread=512
    attenuationShapeExtents {x=1000}
   }
  }
 }
 {id=WavePlayer_0, properties {bLooping=true, soundWaveAssetPtr=&/WilliePete_Ambient_Loop}}
]
properties {
 bOverrideConcurrency=true
 bPrimeOnLoad=true
 maxDistance=4000
 soundClassObject=&Fort_Gadgets_Items
 volumeMultiplier=1
 attenuationOverrides {attenuationShapeExtents {x=0}}
 concurrencyOverrides {maxCount=3, resolutionRule=StopQuietest}
}
