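# Sound cue for entering the Whirlpool secret passage hideable, judging by its name and folder.
# A Random node picks one of two equally weighted WilliePete enter waves, then a Modulator applies slight pitch variation and a fixed volume boost.
# A SoundClass node overrides the cue's Fort_Gadgets_Items class with Fort_Traps, and the root Attenuation node uses the shared traps attenuation.
# The cue primes on load.
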SoundCue /Game/Sounds/Fort_Hideables/SecretPassage/Whirlpool/WilliePete_Enter_Cue
root=Attenuation_0
nodes [
 {id=Attenuation_0, children [Modulator_0], properties {attenuationSettings=&Fort_Traps_Att}}
 {id=Modulator_0, children [SoundClass_0], properties {pitchMax=1.03, pitchMin=0.97, volumeMax=2, volumeMin=2}}
 {id=Random_1, children [WavePlayer_3,WavePlayer_4], properties {weights [1,1]}}
 {id=SoundClass_0, children [Random_1], properties {soundClassOverride=&Fort_Traps}}
 {id=WavePlayer_3, properties {soundWaveAssetPtr=&/WilliePete_Enter_01}}
 {id=WavePlayer_4, properties {soundWaveAssetPtr=&/WilliePete_Enter_02}}
]
properties {bPrimeOnLoad=true, maxDistance=6000, soundClassObject=&Fort_Gadgets_Items, volumeMultiplier=0.5714}
