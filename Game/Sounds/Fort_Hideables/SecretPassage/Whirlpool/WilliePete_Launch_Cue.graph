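# Sound cue for the WilliePete launch sound, filed under the hideable secret-passage whirlpool audio and assigned to the gadgets/items sound class.
# It picks one of two launch waves with equal weight, then a modulator slightly varies and lowers the volume.
# The cue primes on load and references the WilliePete movement attenuation asset, alongside its own falloff, non-spatialized radius and reverb override values.
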SoundCue /Game/Sounds/Fort_Hideables/SecretPassage/Whirlpool/WilliePete_Launch_Cue
root=Modulator_0
nodes [
 {id=Modulator_0, children [Random_0], properties {volumeMax=0.8, volumeMin=0.75}}
 {id=Random_0, children [WavePlayer_1,WavePlayer_2], properties {weights [1,1]}}
 {id=WavePlayer_1, properties {soundWaveAssetPtr=&/WilliePete_Launch_01}}
 {id=WavePlayer_2, properties {soundWaveAssetPtr=&/WilliePete_Launch_02}}
]
properties {
 attenuationSettings=/Game/Sounds/Attenuation/Fort_WilliePete_Movement_Att
 bPrimeOnLoad=true
 maxDistance=14000
 soundClassObject=&Fort_Gadgets_Items
 volumeMultiplier=1.15
 attenuationOverrides {falloffDistance=8000, nonSpatializedRadiusStart=1000, reverbWetLevelMax=0.3, reverbWetLevelMin=0.1, attenuationShapeExtents {x=2000}}
}
